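; Gameplay ability for the Dance Grenade that breaks the owner's dance stun when they take enough damage.
; On event activation it runs the parent logic, then checks for a damage-received event tag with magnitude above a small threshold.
; Passing that check leads to removing effects that grant the DanceStunned tag and ending the ability.
; The end-ability event has no logic of its own.
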(event Ability|EventActivateAbilityFromEvent
  (params (EventData :type Gameplay%20Event%20Data%20Structure%20%28by%20ref%29))
  (bind (_eventtag _instigator _target _optionalobject _optionalobject2 _contexthandle _instigatortags _targettags _eventmagnitude) (Utilities|Struct|BreakGameplayEventData EventData))
  (|CallParentFunction EventData)
  (if (GameplayTags|MatchesTag _eventtag "(TagName=\"Event.Damage.Received\")")
    (if (> _eventmagnitude 6.0)
      _))
  (Ability|RemoveGameplayEffectFromOwnerWithGrantedTags "(GameplayTags=((TagName=\"Gameplay.Action.Player.DanceStunned\")))")
  (Ability|EndAbility))

(event Ability|EventOnEndAbility
  (params (bWasCancelled :type Boolean)))
